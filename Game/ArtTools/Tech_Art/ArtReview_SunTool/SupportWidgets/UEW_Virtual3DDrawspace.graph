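# Editor utility widget for the art-review sun tool that draws a pseudo-3D compass of axes, sun direction and above-horizon sun path.
# Tick builds a virtual plane from editor camera heading, reads FortniteMaterialParameters sun vectors, and rotates the direction around sun Y (overridable) for the path.
# Points project by line-plane intersection, then barycentric quad remapping into widget space for OnPaint.
# Sun Y/Z line macros exist but are unwired.
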EditorUtilityWidgetBlueprint /Game/ArtTools/Tech_Art/ArtReview_SunTool/SupportWidgets/UEW_Virtual3DDrawspace
graphs [
 AddLine {nodes [
   n1 = AddLine!() -> (then, LineStart:Vector, LineEnd:Vector, Color:Linear Color Structure) {is=title K2Node_FunctionEntry}
   > n2 = Add!(
    execute
    TargetArray:Array of Struct Virtual 3DDrawspace Line Structures=Get LinesToDraw() -> (LinesToDraw:Array of Struct Virtual 3DDrawspace Line Structures)
    NewItem_LineStart3D_5_#1:Vector=@n1
    NewItem_LineEnd3D_6_#2:Vector=@n1.LineEnd
    NewItem_LineStart2D_9_#3:Vector 2D Structure=(X=0,Y=0)
    NewItem_LineEnd2D_11_#4:Vector 2D Structure=(X=0,Y=0)
    NewItem_LineColor_14_#5:Linear Color Structure=@n1.Color
   ) -> (then, ReturnValue) {is=title K2Node_CallArrayFunction}
   > Return Node!(execute, Index:int=@n2)
  ]}
 CheckForOverWrittenVector {nodes [
   n1 = CheckForOverWrittenVector!() -> (then, ParameterName:Name) {is=title K2Node_FunctionEntry}
   > n2 = Branch(Condition=Get HasOverWrittenYZ?() -> (HasOverWrittenYZ?:bool))
   > n3 = Switch on Name!(execute:Exec, Selection:Name=@n1) -> (Default:Exec, SunlightYTransformVector:Exec, SunlightZTransformVector:Exec) {is=title K2Node_SwitchName, pinNames [SunlightYTransformVector, SunlightZTransformVector]}
   > PrintString(InString=NAME ERROR IN CHECKFOROVERWRITTENVECTOR)
   Return Node!(execute=@n3.SunlightYTransformVector, Value:Vector=Get OverWrittenY() -> (OverWrittenY:Vector))
   Return Node(execute=@n3.SunlightZTransformVector, Value=Get OverWrittenZ() -> (OverWrittenZ:Vector))
   n4 = GetVectorParameterValue!(
    execute=@n2.else
    WorldContextObject
    Collection:Material Parameter Collection Object Reference=&FortniteMaterialParameters
    ParameterName:Name=@n1
   ) -> (then, ReturnValue_R:float=~, ReturnValue_G:float=~, ReturnValue_B:float=~, ReturnValue_A:float=~) {is=title K2Node_CallMaterialParameterCollectionFunction}
   > Return Node(Value=Normalize!(A:Vector=MakeVector(X=@n4, Y=@n4.ReturnValue_G, Z=@n4.ReturnValue_B), Tolerance:float=0.0001) -> (ReturnValue:Vector) {is=title K2Node_CallFunction})
  ]}
 ConvertFromVirtualSpaceToDrawScreenSpace {nodes [
   n1 = Inputs!() -> (execute, TargetPoint:Vector=~)
   > n2 = Convert SquareBaryCentricIN!(
    execute
    TargetPoint:Vector=@n1
    SourceA:Vector=Get BottomLeft3D() -> (BottomLeft3D:Vector)
    SourceB:Vector=Get UpperLeft3D() -> (UpperLeft3D:Vector)
    SourceC:Vector=Get BottomRight3D() -> (BottomRight3D:Vector)
    SourceD:Vector=Get UpperRight3D() -> (UpperRight3D:Vector)
    __WorldContext:Object Reference
   ) -> (then, UVWCoords:Vector, IsTriangleOne?:bool, PointIsValid?:bool)
   > n3 = EnsureTargetPointIsValidForBaryCentrics!(execute, self, TargetPoint:Vector=@n1, PrecomputedPoint:Vector=@n2, IsPrecomputedValid?:bool=@n2.PointIsValid?) -> (then, ValidTargetPoint:Vector)
   n4 = Get!() -> (BottomLeft_X:double, BottomLeft_Y:double)
   n5 = Get!() -> (UpperLeft_X:double, UpperLeft_Y:double)
   n6 = Get!() -> (UpperRight_X:double, UpperRight_Y:double)
   n7 = Get!() -> (BottomRight_X:double, BottomRight_Y:double)
   > n8 = Convert SquareBaryCentricOUT!(
    execute
    TargetPoint:Vector=@n3
    SourceA:Vector=MakeVector(X=@n4, Y=@n4.BottomLeft_Y)
    SourceB:Vector=MakeVector(X=@n5, Y=@n5.UpperLeft_Y)
    SourceC:Vector=MakeVector(X=@n7, Y=@n7.BottomRight_Y)
    SourceD:Vector=MakeVector(X=@n6, Y=@n6.UpperRight_Y)
    IsTriangleOne?:bool=@n2.IsTriangleOne?
    __WorldContext:Object Reference
   ) -> (then, PositionOut_X:double, PositionOut_Y:double, PositionOut_Z:double)
   > Outputs!(then:Exec, ReturnValue:Vector 2D Structure=MakeVector2D(X=@n8, Y=@n8.PositionOut_Y))
  ]}
 Create2DScreenpointFromVirtualLine {nodes [
   n1 = Inputs!() -> (execute, StartPoint3D:"Vector (by ref)", EndPoint3D:"Vector (by ref)")
   n2 = Get DrawScreen3DCenter() -> (DrawScreen3DCenter:Vector)
   n3 = Get DrawScreen3DEyePosition() -> (DrawScreen3DEyePosition:Vector)
   n4 = Get EditorWindowForwardVector() -> (EditorWindowForwardVector:Vector)
   > n5 = Project3DSpatialPointToDrawPlane!(
    execute
    self
    Spatial3DPoint:"Vector (by ref)"=@n1
    EyePosition:"Vector (by ref)"=@n3
    DrawPlaneOrigin:Vector=@n2
    DrawPlaneNormal:Vector=@n4
   ) -> (then, Intersection:Vector)
   > n6 = Project3DSpatialPointToDrawPlane(Spatial3DPoint=@n1.EndPoint3D, EyePosition=@n3, DrawPlaneOrigin=@n2, DrawPlaneNormal=@n4)
   > n7 = Sequence()
   > n8 = Convert from Virtual Space to Draw Screen Space!(execute:Exec, TargetPoint:Vector=@n5) -> (then:Exec, ReturnValue:Vector 2D Structure) {is=title K2Node_MacroInstance}
   n9 = Convert from Virtual Space to Draw Screen Space(execute=@n7.then_1, TargetPoint=@n6)
   > Outputs!(Out:Exec, StartPoint:Vector 2D Structure=@n8, EndPoint:Vector 2D Structure=@n9)
  ]}
 DrawScreenUpdateAxis {nodes [
   DrawScreenUpdateAxis() {is=title K2Node_FunctionEntry}
   n1 = Get DrawScreen3DCenter()
   n2 = Get DrawScreen3DEyePosition()
   n3 = Get EditorWindowForwardVector()
   > n4 = Project3DSpatialPointToDrawPlane(Spatial3DPoint=MakeVector(Y=0), EyePosition=@n2, DrawPlaneOrigin=@n1, DrawPlaneNormal=@n3)
   > n5 = Project3DSpatialPointToDrawPlane(Spatial3DPoint=MakeVector(Y=0, Z=50), EyePosition=@n2, DrawPlaneOrigin=@n1, DrawPlaneNormal=@n3)
   > n6 = Project3DSpatialPointToDrawPlane(Spatial3DPoint=MakeVector(Y=50), EyePosition=@n2, DrawPlaneOrigin=@n1, DrawPlaneNormal=@n3)
   > n7 = Project3DSpatialPointToDrawPlane(Spatial3DPoint=MakeVector(X=50, Y=0), EyePosition=@n2, DrawPlaneOrigin=@n1, DrawPlaneNormal=@n3)
   > n8 = Sequence!(execute) -> (then_0, then_1, then_2:Exec, then_3:Exec, then_4:Exec)
   > n9 = Convert from Virtual Space to Draw Screen Space(TargetPoint=@n4)
   > Set DrawScreenOrigin(DrawScreenOrigin:Vector 2D Structure=@n9) -> (Output_Get:Vector 2D Structure)
   n10 = Convert from Virtual Space to Draw Screen Space(execute=@n8.then_1, TargetPoint=@n5)
   > Set DrawSceenAxisZ(DrawSceenAxisZ:Vector 2D Structure=@n10) -> (Output_Get:Vector 2D Structure)
   n11 = Convert from Virtual Space to Draw Screen Space(execute=@n8.then_2, TargetPoint=@n6)
   > Set DrawScreenAxisY(DrawScreenAxisY:Vector 2D Structure=@n11) -> (Output_Get:Vector 2D Structure)
   n12 = Convert from Virtual Space to Draw Screen Space(execute=@n8.then_3, TargetPoint=@n7)
   > Set DrawScreenAxisX(DrawScreenAxisX:Vector 2D Structure=@n12) -> (Output_Get:Vector 2D Structure)
   Return Node!(execute=@n8.then_4)
  ]}
 DrawSpaceLinePlaneIntersect {nodes [
   n1 = DrawSpaceLinePlaneIntersect!() -> (then, Plane Normal:Vector=~, Point on Plane:Vector=~, Point on Line:Vector=~) {is=title K2Node_FunctionEntry}
   n2 = LinePlaneIntersection!(
    LineStart:"Vector (by ref)"=@n1.Point on Line
    LineEnd:"Vector (by ref)"=Get DrawScreen3DCenter()
    APlane:"Plane Structure (by ref)"=MakePlaneFromPointAndNormal!(Point:Vector=@n1.Point on Plane, Normal:Vector=@n1) -> (ReturnValue:Plane Structure)
   ) -> (T:float, Intersection:Vector, ReturnValue:bool)
   > Return Node!(execute, THit:double=@n2, Intersection:Vector=@n2.Intersection, Intersects?:bool=@n2.ReturnValue)
  ]}
 DrawSunDirection {nodes [
   Inputs!() -> (execute)
   > n1 = GetVectorParameterValue(
    Collection=&FortniteMaterialParameters
    ParameterName=SunlightDirectionalVector
   ) -> (ReturnValue_R=~, ReturnValue_G=~, ReturnValue_B=~, ReturnValue_A=~)
   n2 = MakeVector(X=@n1, Y=@n1.ReturnValue_G, Z=@n1.ReturnValue_B)
   > n3 = Set!(
    execute
    SunAnimeDirection_LineStart3D_5_#1:Vector=MakeVector(Y=0)
    SunAnimeDirection_LineEnd3D_6_#2:Vector=vector * float(A=Normalize(A=@n2, Tolerance=0.0001), B=60)
    SunAnimeDirection_LineStart2D_9_#3:Vector 2D Structure=(X=0,Y=0)
    SunAnimeDirection_LineEnd2D_11_#4:Vector 2D Structure=(X=0,Y=0)
    SunAnimeDirection_LineColor_14_#5:Linear Color Structure=(R=0.9671,G=1,B=0,A=1)
   ) -> (
    then
    Output_Get_LineStart3D_5_#1:Vector=~
    Output_Get_LineEnd3D_6_#2:Vector=~
    Output_Get_LineStart2D_9_#3:Vector 2D Structure
    Output_Get_LineEnd2D_11_#4:Vector 2D Structure
    Output_Get_LineColor_14_#5:Linear Color Structure
   )
   > n4 = Create 2DScreenpoint from Virtual Line!(execute:Exec, StartPoint3D:"Vector (by ref)"=@n3, EndPoint3D:"Vector (by ref)"=@n3.Output_Get_LineEnd3D_6_#2) -> (Out:Exec, StartPoint:Vector 2D Structure, EndPoint:Vector 2D Structure) {is=title K2Node_MacroInstance}
   > Set members in Struct Virtual 3DDrawspace Line!(
    execute:Exec
    StructRef:"Struct Virtual 3DDrawspace Line Structure (by ref)"=Get SunAnimeDirection() -> (SunAnimeDirection:Struct Virtual 3DDrawspace Line Structure)
    LineStart2D_9_#3:Vector 2D Structure=@n4
    LineEnd2D_11_#4:Vector 2D Structure=@n4.EndPoint
   ) -> (then:Exec, StructOut:"Struct Virtual 3DDrawspace Line Structure (by ref)") {is=title K2Node_SetFieldsInStruct}
   Outputs!()
  ]}
 DrawSunPath {nodes [
   Inputs()
   > n1 = CheckForOverWrittenVector!(execute, self, ParameterName:Name=SunlightYTransformVector) -> (then, Value:Vector) {is=title K2Node_CallFunction}
   > ClearLines(self:Interface Virtual 3DDrawspace Interface)
   > n2 = For Loop(FirstIndex=-20, LastIndex=19)
   Outputs!(Out:Exec=@n2.Completed)
   n3 = Normalize(A=@n1, Tolerance=0.0001)
   n4 = float / float(A=180, B=20) {is=title K2Node_CallFunction}
   n5 = Get!() -> (
    SunAnimeDirection_LineStart3D_5_#1:Vector=~
    SunAnimeDirection_LineEnd3D_6_#2:Vector=~
    SunAnimeDirection_LineStart2D_9_#3:Vector 2D Structure
    SunAnimeDirection_LineEnd2D_11_#4:Vector 2D Structure
    SunAnimeDirection_LineColor_14_#5:Linear Color Structure
   )
   n6 = Rotate Vector Around Axis!(
    InVect:Vector=@n5.SunAnimeDirection_LineEnd3D_6_#2
    AngleDeg:float=int * float!(A:int=int + int(A=@n2, B=1) -> (ReturnValue=0) {is=title K2Node_CommutativeAssociativeBinaryOperator}, B:double=@n4) -> (ReturnValue:double)
    Axis:Vector=@n3
   ) -> (ReturnValue:Vector)
   n7 = BreakVector(InVec=@n6)
   n8 = Rotate Vector Around Axis(InVect=@n5.SunAnimeDirection_LineEnd3D_6_#2, AngleDeg=int * float(A=@n2, B=@n4), Axis=@n3)
   n9 = BreakVector(InVec=@n8)
   > Branch(Condition=AND Boolean(A="float > float"(A=@n9.Z, B=0) {is=title K2Node_CallFunction}, B="float > float"(A=@n7.Z, B=0)))
   > n10 = AddLine!(execute, self, LineStart:Vector=@n8, LineEnd:Vector=@n6, Color:Linear Color Structure=(R=0.8229,G=0.4306,B=0.1518,A=1)) -> (then, Index:int=0) {is=title K2Node_CallFunction}
   > n11 = Create 2DScreenpoint from Virtual Line(StartPoint3D=@n8, EndPoint3D=@n6)
   > Set members in Struct Virtual 3DDrawspace Line(
    StructRef="Get (a ref)"(Array:Array of Struct Virtual 3DDrawspace Line Structures=Get LinesToDraw(), Dimension 1=@n10) -> (Output:"Struct Virtual 3DDrawspace Line Structure (by ref)")
    LineStart2D_9_#3=@n11
    LineEnd2D_11_#4=@n11.EndPoint
   )
  ]}
 DrawSunY {nodes [
   Inputs!() -> (InputPin:Exec)
   > n1 = CheckForOverWrittenVector(ParameterName=SunlightYTransformVector)
   > n2 = Set!(
    execute
    SunY_LineStart3D_5_#1:Vector=MakeVector(Y=0)
    SunY_LineEnd3D_6_#2:Vector=vector * float(A=@n1, B=50)
    SunY_LineStart2D_9_#3:Vector 2D Structure=(X=0,Y=0)
    SunY_LineEnd2D_11_#4:Vector 2D Structure=(X=0,Y=0)
    SunY_LineColor_14_#5:Linear Color Structure=(R=0.294,G=1,B=0.7453,A=1)
   ) -> (
    then
    Output_Get_LineStart3D_5_#1=~
    Output_Get_LineEnd3D_6_#2=~
    Output_Get_LineStart2D_9_#3
    Output_Get_LineEnd2D_11_#4
    Output_Get_LineColor_14_#5
   )
   > n3 = Create 2DScreenpoint from Virtual Line(StartPoint3D=@n2, EndPoint3D=@n2.Output_Get_LineEnd3D_6_#2)
   > Set members in Struct Virtual 3DDrawspace Line(
    StructRef=Get SunY() -> (SunY:Struct Virtual 3DDrawspace Line Structure)
    LineStart2D_9_#3=@n3
    LineEnd2D_11_#4=@n3.EndPoint
   )
   Outputs!()
  ]}
 DrawSunZ {nodes [
   Inputs!() -> (execute:Exec)
   > n1 = CheckForOverWrittenVector(ParameterName=SunlightZTransformVector)
   > n2 = Set!(
    execute
    SunZ_LineStart3D_5_#1:Vector=MakeVector(Y=0)
    SunZ_LineEnd3D_6_#2:Vector=vector * float(A=@n1, B=40)
    SunZ_LineStart2D_9_#3:Vector 2D Structure=(X=0,Y=0)
    SunZ_LineEnd2D_11_#4:Vector 2D Structure=(X=0,Y=0)
    SunZ_LineColor_14_#5:Linear Color Structure=(R=1,G=0.4711,B=0.9813,A=1)
   ) -> (
    then
    Output_Get_LineStart3D_5_#1=~
    Output_Get_LineEnd3D_6_#2=~
    Output_Get_LineStart2D_9_#3
    Output_Get_LineEnd2D_11_#4
    Output_Get_LineColor_14_#5
   )
   > n3 = Create 2DScreenpoint from Virtual Line(StartPoint3D=@n2, EndPoint3D=@n2.Output_Get_LineEnd3D_6_#2)
   > Set members in Struct Virtual 3DDrawspace Line(
    StructRef=Get SunZ() -> (SunZ:Struct Virtual 3DDrawspace Line Structure)
    LineStart2D_9_#3=@n3
    LineEnd2D_11_#4=@n3.EndPoint
   )
   Outputs()
  ]}
 EnsureTargetPointIsValidForBaryCentrics {nodes [
   n1 = EnsureTargetPointIsValidForBaryCentrics!() -> (then, TargetPoint:Vector=~, PrecomputedPoint:Vector=~, IsPrecomputedValid?:bool=~) {is=title K2Node_FunctionEntry}
   > n2 = Branch(Condition=@n1.IsPrecomputedValid?)
   > Return Node!(execute, ValidTargetPoint:Vector=@n1.PrecomputedPoint)
   n3 = ProjectVectorOnToVector!(V:Vector=@n1, Target:Vector=Get DrawScreen3DUP() -> (DrawScreen3DUP:Vector)) -> (ReturnValue:Vector)
   n4 = Get DrawScreen3DLeft() -> (DrawScreen3DLeft:Vector)
   n5 = ProjectVectorOnToVector(V=@n1, Target=@n4)
   n6 = Branch(execute=@n2.else, Condition="float > float"(A=Vector Length(A=@n5), B=Vector Length(A=@n3)))
   > n7 = Branch(Condition="float > float"(A=GetAngleBetweenTwoVectors!(VectorA:Vector=@n4, VectorB:Vector=@n1, __WorldContext:Object Reference) -> ("Angle (Degrees)":double), B=90))
   > PrintString(InString=Left)
   n8 = Get DrawScreen3DLeft()
   > Return Node(ValidTargetPoint=DrawSpaceLinePlaneIntersect!(Plane Normal:Vector=@n8, Point on Plane:Vector=Get UpperLeft3D(), Point on Line:Vector=@n1) -> (THit:double, Intersection:Vector, Intersects?:bool) {is=title K2Node_CallFunction}.Intersection)
   PrintString(execute=@n7.else, InString=right)
   > Return Node(ValidTargetPoint=DrawSpaceLinePlaneIntersect(Plane Normal=@n8, Point on Plane=Get BottomRight3D(), Point on Line=@n1).Intersection)
   n9 = Get DrawScreen3DUP()
   n10 = Branch(execute=@n6.else, Condition="float > float"(A=GetAngleBetweenTwoVectors(VectorA=@n9, VectorB=@n1), B=90))
   > PrintString(InString=Down)
   n11 = Get DrawScreen3DUP()
   > Return Node(ValidTargetPoint=DrawSpaceLinePlaneIntersect(Plane Normal=@n11, Point on Plane=Get BottomRight3D(), Point on Line=@n1).Intersection)
   PrintString(execute=@n10.else, InString=UP)
   > Return Node(ValidTargetPoint=DrawSpaceLinePlaneIntersect(Plane Normal=@n11, Point on Plane=Get UpperLeft3D(), Point on Line=@n1).Intersection)
  ]}
 EventGraph {nodes [
   Event Construct()
   > Set IsBeingUsedInViewport?(IsBeingUsedInViewport?:bool=IsInViewport!(self:Widget Object Reference=Self-Reference()) -> (ReturnValue:bool)) -> (Output_Get:bool)
   Event Tick()
   n1 = GetCachedGeometry!(self:Widget Object Reference=Get DrawSpace() -> (DrawSpace:Border Object Reference)) -> (ReturnValue:"Geometry Structure (by ref)")
   > n2 = Branch(Condition="Equal (Vector2D)"!(
     A:Vector 2D Structure=GetAbsoluteSize!(Geometry:"Geometry Structure (by ref)"=@n1) -> (ReturnValue:Vector 2D Structure)
     B:Vector 2D Structure=Get Diameters() -> (Diameters:Vector 2D Structure)
     ErrorTolerance:float=0.0001
    ) -> (ReturnValue:bool))
   n3 = Get Editor Camera Info!(execute=[@n27.then, @n2.then]) -> (then, CameraLocation:Vector, ForwardDirection_X:double, ForwardDirection_Y:double, ForwardDirection_Z:double, FOV:float, ReturnValue:bool) {is=title K2Node_CallFunction}
   n4 = vector * float(A=Normalize!(A_X:double=@n3.ForwardDirection_X, A_Y:double=@n3.ForwardDirection_Y, A_Z:double=0, Tolerance=0.0001) -> (ReturnValue), B=50)
   n5 = vector + vector(A:Vector="0, 0, 0", B:Vector="0, 0,20") -> (ReturnValue:Vector) {is=title K2Node_CommutativeAssociativeBinaryOperator}
   > Set EditorWindowForwardVector(EditorWindowForwardVector:Vector=Normalize!(A:Vector=vector - vector!(A:Vector=@n4, B:Vector=@n5) -> (ReturnValue:Vector) {is=title K2Node_CallFunction}, Tolerance=0.0001) -> (ReturnValue)) -> (Output_Get:Vector)
   n6 = Get EditorWindowForwardVector()
   n7 = Get DrawSpaceVirtualPlaneDistanceFromOrigin() -> (DrawSpaceVirtualPlaneDistanceFromOrigin:double)
   > Set DrawScreen3DCenter(DrawScreen3DCenter:Vector=vector + vector(A=vector * float(A=@n6, B=@n7), B="0, 0, 0")) -> (Output_Get:Vector)
   > Set DrawScreen3DEyePosition(DrawScreen3DEyePosition:Vector=vector * float(A=Get EditorWindowForwardVector(), B=Get EyeDistanceFromOrigin() -> (EyeDistanceFromOrigin:double))) -> (Output_Get:Vector)
   n8 = Get EditorWindowForwardVector()
   > Set DrawScreen3DLeft(DrawScreen3DLeft:Vector=Normalize(A=Cross Product!(A:Vector="0, 0,1", B:Vector=@n8) -> (ReturnValue:Vector), Tolerance=0.0001)) -> (Output_Get:Vector)
   n9 = Get EditorWindowForwardVector()
   n10 = Get DrawScreen3DLeft()
   > Set DrawScreen3DUP(DrawScreen3DUP:Vector=Normalize(A=Cross Product(A=@n9, B=@n10), Tolerance=0.0001)) -> (Output_Get:Vector)
   n11 = Get VirtualPlaneBaryCentricPointDistanceFromOrigin() -> (VirtualPlaneBaryCentricPointDistanceFromOrigin:double)
   n12 = vector * float(A=Get DrawScreen3DLeft(), B=@n11)
   n13 = vector - vector(A="0, 0, 0", B=@n12)
   n14 = vector * float(A=Get DrawScreen3DUP(), B=@n11)
   n15 = vector - vector(A="0, 0, 0", B=@n14)
   n16 = Get DrawScreen3DCenter()
   > Set BottomLeft3D(BottomLeft3D:Vector=vector + vector!(A=@n15, B=@n13, C:Vector=@n16) -> (ReturnValue)) -> (Output_Get:Vector)
   n17 = vector + vector!(A="0, 0, 0", B=@n14) -> (ReturnValue)
   > Set UpperLeft3D(UpperLeft3D:Vector=vector + vector!(A=@n17, B=@n13, C:Vector=@n16) -> (ReturnValue)) -> (Output_Get:Vector)
   n18 = vector + vector!(A="0, 0, 0", B=@n12) -> (ReturnValue)
   > Set UpperRight3D(UpperRight3D:Vector=vector + vector!(A=@n17, B=@n18, C:Vector=@n16) -> (ReturnValue)) -> (Output_Get:Vector)
   > Set BottomRight3D(BottomRight3D:Vector=vector + vector(A=@n15, B=@n18, C=@n16)) -> (Output_Get:Vector)
   > DrawScreenUpdateAxis() {is=title K2Node_CallFunction}
   > n19 = Sequence!(execute) -> (then_0, then_1, then_2, then_3)
   > Draw Sun Direction!(execute:Exec) {is=title K2Node_MacroInstance}
   Draw Sun Path!(execute:Exec=@n19.then_3) -> (Out:Exec) {is=title K2Node_MacroInstance}
   n20 = GetCachedGeometry(self=Get DrawSpace())
   n21 = GetLocalSize!(Geometry:"Geometry Structure (by ref)"=@n20) -> (ReturnValue:Vector 2D Structure)
   n22 = vector2d / float!(A:Vector 2D Structure=@n21, B:double=2) -> (ReturnValue:Vector 2D Structure)
   Set Center(execute=@n2.else, Center:Vector 2D Structure=@n22) -> (Output_Get:Vector 2D Structure)
   n23 = GetLocalTopLeft!(Geometry:"Geometry Structure (by ref)"=@n20) -> (ReturnValue:Vector 2D Structure)
   > Set UpperLeft(UpperLeft:Vector 2D Structure=@n23) -> (Output_Get:Vector 2D Structure)
   n24 = BreakVector2D(InVec=@n21)
   n25 = BreakVector2D(InVec=@n23)
   > Set BottomLeft(BottomLeft:Vector 2D Structure=MakeVector2D(X=@n25, Y=@n24.Y)) -> (Output_Get:Vector 2D Structure)
   > Set UpperRight(UpperRight:Vector 2D Structure=MakeVector2D(X=@n24, Y=@n25.Y)) -> (Output_Get:Vector 2D Structure)
   > Set BottomRight(BottomRight:Vector 2D Structure=@n21) -> (Output_Get:Vector 2D Structure)
   > Set Diameters(Diameters:Vector 2D Structure=@n21) -> (Output_Get:Vector 2D Structure)
   > Set Radii(Radii:Vector 2D Structure=@n22) -> (Output_Get:Vector 2D Structure)
   n26 = BreakVector2D(InVec=@n22)
   > n27 = Set SmallestRadius(SmallestRadius:double="Min (Float)"(A:double=@n26, B:double=@n26.Y) -> (ReturnValue:double) {is=title K2Node_CommutativeAssociativeBinaryOperator}) -> (Output_Get:double)
   Event ClearLines() {is=title K2Node_Event}
   > Clear(TargetArray:Array of Struct Virtual 3DDrawspace Line Structures=Get LinesToDraw())
   n28 = Event EditLine!() -> (OutputDelegate, then, Index:int=0, LineStart:Vector, LineEnd:Vector, Color:Linear Color Structure) {is=title K2Node_Event}
   > Set members in Struct Virtual 3DDrawspace Line!(
    execute
    StructRef="Get (a ref)"(Array=Get LinesToDraw(), Dimension 1=@n28.Index)
    LineStart3D_5_#1:Vector=@n28.LineStart
    LineEnd3D_6_#2:Vector=@n28.LineEnd
    LineColor_14_#5:Linear Color Structure=@n28.Color
   ) -> (then, StructOut)
   n29 = Event RemoveLine!() -> (OutputDelegate, then, Index:int=0) {is=title K2Node_Event}
   > Remove Index!(execute, TargetArray:Array of Struct Virtual 3DDrawspace Line Structures=Get LinesToDraw(), IndexToRemove:int=@n29.Index) -> (then)
   n30 = Event UpdateOverWritten!() -> (OutputDelegate, then, IsOverWritten?:bool, SunY:Vector, SunZ:Vector) {is=title K2Node_Event}
   > Set HasOverWrittenYZ?(HasOverWrittenYZ?:bool=@n30.IsOverWritten?) -> (Output_Get:bool)
   > Set OverWrittenY(OverWrittenY:Vector=@n30.SunY) -> (Output_Get:Vector)
   > Set OverWrittenZ(OverWrittenZ:Vector=@n30.SunZ) -> (Output_Get:Vector)
   EnsureTargetPointIsValidForBaryCentrics(TargetPoint="0, 0, 0", PrecomputedPoint="0, 0, 0", IsPrecomputedValid?=false) {is=title K2Node_CallFunction}
   Event PreConstruct()
   Draw Sun Y!(InputPin:Exec) {is=title K2Node_MacroInstance}
   Draw Sun Z!(execute:Exec) {is=title K2Node_MacroInstance}
   Convert from Virtual Space to Draw Screen Space()
  ]}
 OnPaint {nodes [
   n1 = OnPaint!() -> (then, Context:"Paint Context Structure (by ref)") {is=title K2Node_FunctionEntry}
   > n2 = Sequence()
   n3 = Get() -> (SunAnimeDirection_LineStart3D_5_#1=~, SunAnimeDirection_LineEnd3D_6_#2=~)
   DrawLine!(
    execute=@n2.then_1
    Context:"Paint Context Structure (by ref)"=@n1
    PositionA:Vector 2D Structure=@n3.SunAnimeDirection_LineStart2D_9_#3
    PositionB:Vector 2D Structure=@n3.SunAnimeDirection_LineEnd2D_11_#4
    Tint:Linear Color Structure=@n3.SunAnimeDirection_LineColor_14_#5
    bAntiAlias:bool=true
    Thickness:float=2
   ) -> (then)
   Return Node!(execute=@n2.then_3)
   n4 = Get!() -> (
    SunY_LineStart3D_5_#1:Vector=~
    SunY_LineEnd3D_6_#2:Vector=~
    SunY_LineStart2D_9_#3:Vector 2D Structure
    SunY_LineEnd2D_11_#4:Vector 2D Structure
    SunY_LineColor_14_#5:Linear Color Structure
   )
   DrawLine(
    execute=@n2.then_2
    Context=@n1
    PositionA=@n4.SunY_LineStart2D_9_#3
    PositionB=@n4.SunY_LineEnd2D_11_#4
    Tint=@n4.SunY_LineColor_14_#5
    bAntiAlias=true
    Thickness=1
   )
   n5 = Get!() -> (
    SunZ_LineStart3D_5_#1:Vector=~
    SunZ_LineEnd3D_6_#2:Vector=~
    SunZ_LineStart2D_9_#3:Vector 2D Structure
    SunZ_LineEnd2D_11_#4:Vector 2D Structure
    SunZ_LineColor_14_#5:Linear Color Structure
   )
   > DrawLine(
    Context=@n1
    PositionA=@n5.SunZ_LineStart2D_9_#3
    PositionB=@n5.SunZ_LineEnd2D_11_#4
    Tint=@n5.SunZ_LineColor_14_#5
    bAntiAlias=true
    Thickness=1
   )
   > n6 = For Each Loop!(Exec, Array:Array of Struct Virtual 3DDrawspace Line Structures=Get LinesToDraw()) -> (
    LoopBody
    Array Element_LineStart3D_5_#1:Vector=~
    Array Element_LineEnd3D_6_#2:Vector=~
    Array Element_LineStart2D_9_#3:Vector 2D Structure
    Array Element_LineEnd2D_11_#4:Vector 2D Structure
    Array Element_LineColor_14_#5:Linear Color Structure
    Array Index
    Completed
   )
   > DrawLine(
    Context=@n1
    PositionA=@n6.Array Element_LineStart2D_9_#3
    PositionB=@n6.Array Element_LineEnd2D_11_#4
    Tint=@n6.Array Element_LineColor_14_#5
    bAntiAlias=true
    Thickness=1
   )
   n7 = Get DrawScreenOrigin() -> (DrawScreenOrigin:Vector 2D Structure)
   DrawLine(
    execute=@n2.then_0
    Context=@n1
    PositionA=@n7
    PositionB=Get DrawSceenAxisZ() -> (DrawSceenAxisZ:Vector 2D Structure)
    Tint=(R=0,G=0,B=1,A=1)
    bAntiAlias=true
    Thickness=1
   )
   > DrawLine(
    Context=@n1
    PositionA=Get DrawScreenAxisY() -> (DrawScreenAxisY:Vector 2D Structure)
    PositionB=@n7
    Tint=(R=0,G=1,B=0,A=1)
    bAntiAlias=true
    Thickness=1
   )
   > DrawLine(
    Context=@n1
    PositionA=@n7
    PositionB=Get DrawScreenAxisX() -> (DrawScreenAxisX:Vector 2D Structure)
    Tint=(R=1,G=0,B=0,A=1)
    bAntiAlias=true
    Thickness=1
   )
   DrawLine(PositionB=Get BottomRight() -> (BottomRight:Vector 2D Structure), Tint=(R=1,G=0.2226,B=0.9687,A=1), bAntiAlias=true, Thickness=1)
  ]}
 Project3DSpatialPointToDrawPlane {nodes [
   n1 = Project3DSpatialPointToDrawPlane!() -> (then, Spatial3DPoint:"Vector (by ref)", EyePosition:"Vector (by ref)", DrawPlaneOrigin:Vector=~, DrawPlaneNormal:Vector=~) {is=title K2Node_FunctionEntry}
   n2 = "Line Plane Intersection (Origin && Normal)"!(
    LineStart:"Vector (by ref)"=@n1
    LineEnd:"Vector (by ref)"=@n1.EyePosition
    PlaneOrigin:Vector=@n1.DrawPlaneOrigin
    PlaneNormal:Vector=@n1.DrawPlaneNormal
   ) -> (T:float, Intersection:Vector, ReturnValue:bool)
   > n3 = Branch(Condition=@n2.ReturnValue)
   > Return Node!(execute, Intersection:Vector=@n2.Intersection)
   n4 = DistanceBetweenTwoVectors!(execute=@n3.else, vector1:Vector=@n1, Vector2:Vector=@n1.DrawPlaneOrigin, __WorldContext:Object Reference) -> (then, Distance:double)
   > n5 = Branch()
   n6 = vector * float(A=vector - vector(A=@n1, B=@n1.EyePosition), B=@n4)
   Return Node(
    execute=[@n7.then, @n5.then]
    Intersection="Line Plane Intersection (Origin && Normal)"(LineStart=@n1, LineEnd=vector * float(A=@n6, B=2), PlaneOrigin=@n1.DrawPlaneOrigin, PlaneNormal=@n1.DrawPlaneNormal).Intersection
   )
   n7 = PrintString(execute=@n5.else, InString=UNKNOWN ERROR in Project3DSpatialPointToDrawPlane: no intersections found)
  ]}
]
variables [
 BottomLeft {type=Vector2D, default=<0 0>}
 BottomLeft3D {type=Vector}
 BottomRight {type=Vector2D, default=<0 0>}
 BottomRight3D {type=Vector}
 Center {type=Vector2D, default=<0 0>}
 Diameters {type=Vector2D, default=<0 0>}
 DrawSceenAxisZ {type=Vector2D, default=<0 0>}
 DrawScreen3DCenter {type=Vector}
 DrawScreen3DEyePosition {type=Vector}
 DrawScreen3DLeft {type=Vector}
 DrawScreen3DUP {type=Vector, default=<0 0 1>}
 DrawScreenAxisX {type=Vector2D, default=<0 0>}
 DrawScreenAxisY {type=Vector2D, default=<0 0>}
 DrawScreenOrigin {type=Vector2D, default=<0 0>}
 DrawSpaceVirtualPlaneDistanceFromOrigin {type=number, default=-500}
 EditorWindowForwardVector {type=Vector}
 EyeDistanceFromOrigin {type=number, default=-1010}
 HasOverWrittenYZ? {type=boolean}
 IsBeingUsedInViewport? {type=boolean}
 LinesToDraw {type="array<Struct_Virtual3DDrawspace_Line>", default []}
 OverWrittenY {type=Vector}
 OverWrittenZ {type=Vector}
 Radii {type=Vector2D, default=<0 0>}
 SmallestRadius {type=number}
 SunAnimeDirection {
  type=Struct_Virtual3DDrawspace_Line
  default {lineColor=<rgba 0 0 0 0>, lineEnd2D=<0 0>, lineEnd3D=<0 0 0>, lineStart2D=<0 0>, lineStart3D=<0 0 0>}
 }
 SunY {
  type=Struct_Virtual3DDrawspace_Line
  default {lineColor=<rgba 0 0 0 0>, lineEnd2D=<0 0>, lineEnd3D=<0 0 0>, lineStart2D=<0 0>, lineStart3D=<0 0 0>}
 }
 SunZ {
  type=Struct_Virtual3DDrawspace_Line
  default {lineColor=<rgba 0 0 0 0>, lineEnd2D=<0 0>, lineEnd3D=<0 0 0>, lineStart2D=<0 0>, lineStart3D=<0 0 0>}
 }
 UpperLeft {type=Vector2D, default=<0 0>}
 UpperLeft3D {type=Vector}
 UpperRight {type=Vector2D, default=<0 0>}
 UpperRight3D {type=Vector}
 VirtualPlaneBaryCentricPointDistanceFromOrigin {type=number, default=32}
]
widgets {
 info {rootWidgetClass=&Overlay, widgetCount=4}
 widgets [
  Overlay_37 {widgetClassPath=&Overlay, widget=Overlay_37}
  HorizontalBox_0 {
   widgetClassPath=&HorizontalBox
   parent=Overlay_37
   widget=HorizontalBox_0
   slot {class=OverlaySlot, properties {horizontalAlignment=HAlign_Fill, verticalAlignment=VAlign_Fill}}
  }
  VerticalBox_0 {
   widgetClassPath=&VerticalBox
   parent=HorizontalBox_0
   widget=VerticalBox_0
   slot {class=HorizontalBoxSlot, properties {size {sizeRule=Fill, value=1}}}
  }
  DrawSpace {
   widgetClassPath=&Border
   bIsVariable=true
   parent=VerticalBox_0
   widget=DrawSpace
   properties {brushColor=<rgba 0.495 0.495 0.495 0>}
   slot {class=VerticalBoxSlot, properties {size {sizeRule=Fill, value=1}}}
  }
 ]
}
