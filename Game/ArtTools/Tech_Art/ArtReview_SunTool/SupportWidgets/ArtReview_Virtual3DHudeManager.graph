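# Editor utility widget for the art review sun tool that hosts and manages a virtual 3D drawspace HUD overlay.
# AddLine returns a line index, while events forward edit, remove, clear and sun-axis override updates to the embedded UEW_Virtual3DDrawspace.
# UpdateSpacers sets Fill weights on four surrounding spacers to position and size the drawspace proportionally.
# The widget switcher has only an unconnected GetChildAt node, and no line state is stored here.
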EditorUtilityWidgetBlueprint /Game/ArtTools/Tech_Art/ArtReview_SunTool/SupportWidgets/ArtReview_Virtual3DHudeManager
graphs [
 AddLine {nodes [
   n1 = AddLine!() -> (then, LineStart:Vector, LineEnd:Vector, Color:Linear Color Structure) {is=title K2Node_FunctionEntry}
   > n2 = AddLine!(
    execute
    self:UEW Virtual 3DDrawspace Object Reference=Get UEW_Virtual3DDrawspace() -> (UEW_Virtual3DDrawspace:UEW Virtual 3DDrawspace Object Reference)
    LineStart:Vector=@n1
    LineEnd:Vector=@n1.LineEnd
    Color:Linear Color Structure=@n1.Color
   ) -> (then, Index:int=0) {is=title K2Node_CallFunction}
   > Return Node!(execute, Index:int=@n2)
  ]}
 EventGraph {nodes [
   n1 = Event UpdateSpacers!() -> (OutputDelegate, then, LeftSpacer:double, RightSpacer:double, TopSpacer:double, BottomSpacer:double) {is=title K2Node_Event}
   > SetSize!(
    execute
    self:Horizontal Box Slot Object Reference=SlotAsHorizontalBoxSlot!(Widget:Widget Object Reference=Get Spacer_Horizontal_Left() -> (Spacer_Horizontal_Left:Spacer Object Reference)) -> (ReturnValue:Horizontal Box Slot Object Reference)
    InSize_Value:float=@n1.LeftSpacer
    InSize_SizeRule:ESlateSizeRule Enum=Fill
   ) -> (then)
   > SetSize(
    self=SlotAsHorizontalBoxSlot(Widget=Get Spacer_Horizontal_Right() -> (Spacer_Horizontal_Right:Spacer Object Reference))
    InSize_Value=@n1.RightSpacer
    InSize_SizeRule=Fill
   )
   > SetSize(
    self:Vertical Box Slot Object Reference=SlotAsVerticalBoxSlot(Widget=Get Spacer_Vertical_Bottom() -> (Spacer_Vertical_Bottom:Spacer Object Reference))
    InSize_Value=@n1.BottomSpacer
    InSize_SizeRule=Fill
   )
   > SetSize(
    self=SlotAsVerticalBoxSlot(Widget=Get Spacer_Vertical_Top() -> (Spacer_Vertical_Top:Spacer Object Reference))
    InSize_Value=@n1.TopSpacer
    InSize_SizeRule=Fill
   )
   Event ClearLines() {is=title K2Node_Event}
   n2 = Get UEW_Virtual3DDrawspace()
   > ClearLines(self:UEW Virtual 3DDrawspace Object Reference=@n2)
   n3 = Event EditLine!() -> (OutputDelegate, then, Index:int=0, LineStart:Vector, LineEnd:Vector, Color:Linear Color Structure) {is=title K2Node_Event}
   > EditLine!(
    execute
    self:UEW Virtual 3DDrawspace Object Reference=@n2
    Index:int=@n3.Index
    LineStart:Vector=@n3.LineStart
    LineEnd:Vector=@n3.LineEnd
    Color:Linear Color Structure=@n3.Color
   ) -> (then)
   n4 = Event RemoveLine!() -> (OutputDelegate, then, Index:int=0) {is=title K2Node_Event}
   > RemoveLine!(execute, self:UEW Virtual 3DDrawspace Object Reference=@n2, Index:int=@n4.Index) -> (then)
   n5 = Event UpdateOverWritten!() -> (OutputDelegate, then, IsOverWritten?:bool, SunY:Vector, SunZ:Vector) {is=title K2Node_Event}
   > UpdateOverWritten!(
    execute
    self:UEW Virtual 3DDrawspace Object Reference=@n2
    IsOverWritten?:bool=@n5.IsOverWritten?
    SunY:Vector=@n5.SunY
    SunZ:Vector=@n5.SunZ
   ) -> (then)
   GetChildAt!(
    self:Panel Widget Object Reference=Get CommonWidgetSwitcherLegacy_276() -> (CommonWidgetSwitcherLegacy_276:Common Widget Switcher Legacy Object Reference)
    Index:int=0
   ) -> (ReturnValue:Widget Object Reference)
   Event PreConstruct()
   Event Construct()
   Event Tick()
  ]}
]
variables []
widgets {
 info {rootWidgetClass=&Overlay, widgetCount=9}
 widgets [
  Overlay_37 {widgetClassPath=&Overlay, widget=Overlay_37}
  HorizontalBox_0 {
   widgetClassPath=&HorizontalBox
   parent=Overlay_37
   widget=HorizontalBox_0
   slot {class=OverlaySlot, properties {horizontalAlignment=HAlign_Fill, verticalAlignment=VAlign_Fill}}
  }
  Spacer_Horizontal_Left {
   widgetClassPath=&Spacer
   bIsVariable=true
   parent=HorizontalBox_0
   widget=Spacer_Horizontal_Left
   slot {class=HorizontalBoxSlot, properties {size {sizeRule=Fill, value=0}}}
  }
  VerticalBox_0 {
   widgetClassPath=&VerticalBox
   parent=HorizontalBox_0
   widget=VerticalBox_0
   slot {class=HorizontalBoxSlot, properties {size {sizeRule=Fill, value=1}}}
  }
  Spacer_Vertical_Top {
   widgetClassPath=&Spacer
   bIsVariable=true
   parent=VerticalBox_0
   widget=Spacer_Vertical_Top
   slot {class=VerticalBoxSlot, properties {size {sizeRule=Fill, value=0}}}
  }
  UEW_Virtual3DDrawspace {
   widgetClassPath=&/UEW_Virtual3DDrawspace
   bIsVariable=true
   parent=VerticalBox_0
   widget=UEW_Virtual3DDrawspace
   slot {class=VerticalBoxSlot, properties {size {sizeRule=Fill, value=1}}}
  }
  CommonWidgetSwitcherLegacy_276 {
   widgetClassPath=&CommonWidgetSwitcherLegacy
   bIsVariable=true
   parent=VerticalBox_0
   widget=CommonWidgetSwitcherLegacy_276
   slot {class=VerticalBoxSlot}
  }
  Spacer_Vertical_Bottom {
   widgetClassPath=&Spacer
   bIsVariable=true
   parent=VerticalBox_0
   widget=Spacer_Vertical_Bottom
   slot {class=VerticalBoxSlot, properties {size {sizeRule=Fill, value=0}}}
  }
  Spacer_Horizontal_Right {
   widgetClassPath=&Spacer
   bIsVariable=true
   parent=HorizontalBox_0
   widget=Spacer_Horizontal_Right
   slot {class=HorizontalBoxSlot, properties {size {sizeRule=Fill, value=0}}}
  }
 ]
}
