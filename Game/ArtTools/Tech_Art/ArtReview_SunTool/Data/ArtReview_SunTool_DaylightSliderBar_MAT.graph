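# UI material for the daylight slider bar in the art review sun tool.
# Horizontal UI UV samples two curve atlas rows, with Rayleigh scatter color and composite sun color filling opposite vertical halves of the bar.
# An absolute sine of horizontal UV, thresholded with Ceil, creates thin repeating tick marks that blend in a vector parameter color over the gradients.
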Material /Game/ArtTools/Tech_Art/ArtReview_SunTool/Data/ArtReview_SunTool_DaylightSliderBar_MAT
graph {
 properties {materialDomain=MD_UI, pixelDepthOffsetMode=PDOM_Legacy, usedShadingModels=Default Lit}
 nodes [
  n1 = MaterialFunctionCall() {materialFunction=&GetUserInterfaceUV}
  n2 = ComponentMask(None=@n1.Normalized UV) {r=true}
  n3 = CurveAtlasRowParameter(CurveTime=@n2) {
   atlas=&/ArtReview_SunTool_DayLightCurve_ATLAS
   curve=&/ArtReview_RayleighScatterData
   defaultValue=2
   parameterName=RayleighColor
  }
  n4 = CurveAtlasRowParameter(CurveTime=@n2) {
   atlas=&ArtReview_SunTool_DayLightCurve_ATLAS
   curve=&/ArtReview_CompositSunData
   defaultValue=1
   parameterName=DLightColor
  }
  n5 = Multiply(A=Add(A=@n2) {constB=0.5}) {constB=0.5}
  n6 = Abs(None=Sine(None=Multiply(A=@n5) {constB=23}))
  LinearInterpolate(
   A=If(A=ComponentMask(None=@n1.Normalized UV) {g=true}, "A > B"=@n3, "A < B"=@n4) {constB=0.5}
   B=VectorParameter().RGB
   Alpha=Ceil(None=Subtract(A=@n6) {constB=0.995})
  )
 ]
}
