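# UI-domain material for the daylight slider selector in the ArtReview SunTool, an art review tool.
# The graph holds a single call to the shared effects function CurveColor_Time_of_Day, which by its name supplies a time-of-day curve color.
# No other nodes, parameters or connections to material outputs appear in the graph.
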Material /Game/ArtTools/Tech_Art/ArtReview_SunTool/Data/ArtReview_SunTool_DaylightSliderSelector_MAT
graph {
 properties {materialDomain=MD_UI, pixelDepthOffsetMode=PDOM_Legacy, usedShadingModels=Default Lit}
 nodes [MaterialFunctionCall() {materialFunction=/Game/Effects/Fort_Effects/Materials/_Functions/CurveColor_Time_of_Day}]
}
