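# Editor utility widget for the Art Review Sun Tool, with time-of-day, sun rotation and inclination sliders.
# Time of day sets the hour through helper actors found in PIE worlds.
# Rotation or inclination switches to a false sun: TODM off, vectors read from FortniteMaterialParameters, cross-product axes pushed as overrides, time slider disabled until reset.
# OnPaint draws day-phase markers and labels over the material-styled time bar.
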EditorUtilityWidgetBlueprint /Game/ArtTools/Tech_Art/ArtReview_SunTool/SupportWidgets/ArtReview_SunTool_RealSun
graphs [
 EventGraph {nodes [
   n1 = Get SunColors() -> (SunColors:Array of Linear Color Structures)
   Clear(TargetArray:Array of Linear Color Structures=@n1)
   > n2 = For Loop(LastIndex=999)
   > Add!(
    execute
    TargetArray:Array of Linear Color Structures=@n1
    NewItem_R:float=float / float(A="To Float (Integer)"(InInt=@n2), B=999) {is=title K2Node_CallFunction}
    NewItem_G:float=0
    NewItem_B:float=0
    NewItem_A:float=1
   ) -> (then, ReturnValue) {is=title K2Node_CallArrayFunction}
   n3 = "On Value Changed (Slider_TimeOfDay)"!() -> (OutputDelegate, then, Value:float) {is=title K2Node_ComponentBoundEvent}
   > UpdateSun!(execute, self, CallingSlider:ENUM_ArtReview_Sliders Enum=NewEnumerator2, SliderValue:double=@n3.Value) -> (then)
   n4 = "On Value Changed (Slider_Rotation)"!() -> (OutputDelegate, then, Value:float) {is=title K2Node_ComponentBoundEvent}
   > UpdateSun(CallingSlider=NewEnumerator0, SliderValue=@n4.Value)
   n5 = "On Value Changed (Slider_Inclination)"!() -> (OutputDelegate, then, Value:float) {is=title K2Node_ComponentBoundEvent}
   > UpdateSun(CallingSlider=NewEnumerator1, SliderValue=@n5.Value)
   n6 = "On Clicked (Button_ResetDirOverToSun)"!() -> (OutputDelegate, then, widget:User Widget Object Reference, event:Pointer Event Structure) {is=title K2Node_ComponentBoundEvent}
   TODM_On(
    execute=[@n6.then, @n7.then]
    self:BP Art Review Helper Chris Object Reference=[
     Get TODM BP Helper Client Ref() -> (TODM BP Helper Client Ref:BP Art Review Helper Chris Object Reference)
     Get TODM BP Helper Dedicated Server Ref() -> (TODM BP Helper Dedicated Server Ref:BP Art Review Helper Chris Object Reference)
    ]
   )
   > Set UsingFalseSun?(UsingFalseSun?:bool=false) -> (Output_Get:bool)
   > Set Enabled And Tooltip Text!(
    execute
    Widget:Widget Object Reference=Get Button_ResetDirOverToSun() -> (Button_ResetDirOverToSun:TA Generic 01 Button Object Reference)
    Enabled:bool=false
    Tooltip Text:Text="LOCTEXT(\"Real Sun Active, Currently Disabled\")"
    __WorldContext:Object Reference
   ) -> (then) {is=title K2Node_CallFunction}
   > SetManualAnimeSunDirection!(
    execute
    self:BR Art Review TODMSystem Object Reference=Get ToDManagerRef() -> (ToDManagerRef:BR Art Review TODMSystem Object Reference)
    SwitchManualAnimeShadeSunDirection:bool=false
   ) -> (then)
   > SetValue!(execute, self:Slider Object Reference=Get Slider_Rotation() -> (Slider_Rotation:Slider Object Reference), InValue:float=0) -> (then)
   > Set Enabled And Tooltip Text(Widget=Get Slider_TimeOfDay() -> (Slider_TimeOfDay:Slider Object Reference), Enabled=true, Tooltip Text="LOCTEXT(\"Scrub Time Of Day\")")
   n7 = ResetToRealSun() {is=title K2Node_CustomEvent}
   Event Construct()
   > Set Enabled And Tooltip Text(Widget=Get Button_ResetDirOverToSun(), Enabled=false, Tooltip Text="LOCTEXT(\"Real Sun is Already Active\")")
   n8 = Make Slate Brush!(
    TintColor:Slate Color Structure=(SpecifiedColor=(R=1,G=1,B=1,A=1),ColorUseRule=UseColor_Specified)
    DrawAs:ESlateBrushDrawType Enum=Image
    Tiling:ESlateBrushTileType Enum=NoTile
    Mirroring:ESlateBrushMirrorType Enum=NoMirror
    ImageSize:Vector 2D Structure=(X=256,Y=32)
    Margin:Margin Structure=(Left=0,Top=0,Right=0,Bottom=0)
    ResourceObject:Object Reference=/Game/ArtTools/Tech_Art/ArtReview_SunTool/Data/ArtReview_SunTool_DaylightSliderBar_MAT
    OutlineSettings:Slate Brush Outline Settings Structure=(CornerRadii=(X=0,Y=0,Z=0,W=0),Color=(SpecifiedColor=(R=0,G=0,B=0,A=0),ColorUseRule=UseColor_Specified),Width=0,RoundingType=HalfHeightRadius,bUseBrushTransparency=False)
   ) -> (SlateBrush:Slate Brush Structure) {is=title K2Node_MakeStruct}
   n9 = Make Slate Brush(
    TintColor=(SpecifiedColor=(R=1,G=1,B=1,A=1),ColorUseRule=UseColor_Specified)
    DrawAs=Image
    Tiling=NoTile
    Mirroring=NoMirror
    ImageSize=(X=16,Y=32)
    Margin=(Left=0,Top=0,Right=0,Bottom=0)
    ResourceObject=/Game/ArtTools/Tech_Art/ArtReview_SunTool/Data/ArtReview_SunTool_DaylightSliderSelector_MAT
    OutlineSettings=(CornerRadii=(X=0,Y=0,Z=0,W=0),Color=(SpecifiedColor=(R=0,G=0,B=0,A=0),ColorUseRule=UseColor_Specified),Width=0,RoundingType=HalfHeightRadius,bUseBrushTransparency=False)
   )
   > Set WidgetStyle!(
    execute
    WidgetStyle:Slider Style Structure=Make Slider Style!(
     NormalBarImage:Slate Brush Structure=@n8
     HoveredBarImage:Slate Brush Structure=@n8
     DisabledBarImage:Slate Brush Structure=(TintColor=(SpecifiedColor=(R=1,G=1,B=1,A=1),ColorUseRule=UseColor_Specified),DrawAs=Image,Tiling=NoTile,Mirroring=NoMirror,ImageType=NoImage,ImageSize=(X=32,Y=32),Margin=(Left=0,Top=0,Right=0,Bottom=0),ResourceObject=None,OutlineSettings=(CornerRadii=(X=0,Y=0,Z=0,W=1),Color=(SpecifiedColor=(R=0,G=0,B=0,A=0),ColorUseRule=UseColor_Specified),Width=0,RoundingType=HalfHeightRadius,bUseBrushTransparency=False),UVRegion=(Min=(X=0,Y=0),Max=(X=0,Y=0),bIsValid=False),bIsDynamicallyLoaded=False,ResourceName="")
     NormalThumbImage:Slate Brush Structure=@n9
     HoveredThumbImage:Slate Brush Structure=@n9
     DisabledThumbImage:Slate Brush Structure=(TintColor=(SpecifiedColor=(R=1,G=1,B=1,A=1),ColorUseRule=UseColor_Specified),DrawAs=Image,Tiling=NoTile,Mirroring=NoMirror,ImageType=NoImage,ImageSize=(X=32,Y=32),Margin=(Left=0,Top=0,Right=0,Bottom=0),ResourceObject=None,OutlineSettings=(CornerRadii=(X=0,Y=0,Z=0,W=1),Color=(SpecifiedColor=(R=0,G=0,B=0,A=0),ColorUseRule=UseColor_Specified),Width=0,RoundingType=HalfHeightRadius,bUseBrushTransparency=False),UVRegion=(Min=(X=0,Y=0),Max=(X=0,Y=0),bIsValid=False),bIsDynamicallyLoaded=False,ResourceName="")
     BarThickness:float=32
    ) -> (SliderStyle:Slider Style Structure) {is=title K2Node_MakeStruct}
    self:Slider Object Reference=Get Slider_TimeOfDay()
   ) -> (then, Output_Get:Slider Style Structure)
   n10 = Get SunColors()
   PrintString(
    InString="To String (LinearColor)"!(InColor:Linear Color Structure="Get (a ref)"(Array:Array of Linear Color Structures=@n10, Dimension 1=@n2) -> (Output:"Linear Color Structure (by ref)")) -> (ReturnValue:String)
    bPrintToScreen=false
   )
   Event PreConstruct()
  ]}
 GenerateLineAlongBarMacro {nodes [
   n1 = Inputs!() -> (A:Vector 2D Structure, B:Vector 2D Structure, A2:Vector 2D Structure, B2:Vector 2D Structure, Alpha:double=~)
   Outputs!(
    ReturnValue:Vector 2D Structure=Lerp Vector 2D!(A:Vector 2D Structure=@n1, B:Vector 2D Structure=@n1.B, Alpha:double=@n1.Alpha, __WorldContext:Object Reference) -> (ReturnValue:Vector 2D Structure)
    ReturnValue2:Vector 2D Structure=Lerp Vector 2D(A=@n1.A2, B=@n1.B2, Alpha=@n1.Alpha)
   )
  ]}
 GetAzimuthFromDirection {nodes [
   n1 = GetAzimuthFromDirection!() -> (then, SunY:Vector=~) {is=title K2Node_FunctionEntry}
   n2 = Normalize!(A:Vector=@n1, Tolerance:float=0.0001) -> (ReturnValue:Vector) {is=title K2Node_CallFunction}
   > n3 = Branch(Condition="float < float"!(A:double=BreakVector(InVec=@n2).Y, B:double=0) -> (ReturnValue:bool) {is=title K2Node_CallFunction})
   n4 = GetAngleBetweenTwoVectors!(VectorA:Vector=@n2, VectorB:Vector="1, 0, 0", __WorldContext:Object Reference) -> ("Angle (Degrees)":double)
   > Return Node!(
    execute
    "Angle (Degrees)":double=float + float(A=float - float(A=90, B=@n4) {is=title K2Node_CallFunction}, B=180) {is=title K2Node_CommutativeAssociativeBinaryOperator}
   )
   Return Node(execute=@n3.else, "Angle (Degrees)"=@n4)
  ]}
 OnPaint {nodes [
   n1 = OnPaint!() -> (then, Context:"Paint Context Structure (by ref)") {is=title K2Node_FunctionEntry}
   > n2 = Branch(Condition=Get HideLabels() -> (HideLabels:bool))
   n3 = Sequence(execute=[@n12.then, @n2.then])
   n4 = MakeVector2D(X=0, Y=Make Literal Float(Value=25))
   n5 = GetLocalTopLeft!(Geometry:"Geometry Structure (by ref)"=GetCachedGeometry!(self:Widget Object Reference=Get Slider_TimeOfDay()) -> (ReturnValue:"Geometry Structure (by ref)")) -> (ReturnValue:Vector 2D Structure)
   n6 = BreakVector2D(InVec=@n5)
   n7 = GetLocalSize!(Geometry:"Geometry Structure (by ref)"=GetCachedGeometry(self=Get HorizontalBox_SunHourControls() -> (HorizontalBox_SunHourControls:Horizontal Box Object Reference))) -> (ReturnValue:Vector 2D Structure)
   n8 = BreakVector2D(InVec=@n7)
   n9 = vector2d + vector2d(A:Vector 2D Structure=vector2d + vector2d(A=MakeVector2D(X=@n6, Y=@n8.Y), B=(X=10,Y=0)), B:Vector 2D Structure=@n4) -> (ReturnValue:Vector 2D Structure) {is=title K2Node_CommutativeAssociativeBinaryOperator}
   n10 = vector2d - vector2d!(A:Vector 2D Structure=@n9, B:Vector 2D Structure=@n4) -> (ReturnValue:Vector 2D Structure)
   n11 = Create SliderLabel!(
    execute=@n2.else
    OnPaintContext:Paint Context Structure=@n1
    Slider:Slider Object Reference=Get Slider_Rotation()
    PositionalShift:Vector 2D Structure=(X=5,Y=-8)
    Label:String=Rotate Sun Around Z
    Label Color:Linear Color Structure=(R=1,G=1,B=1,A=1)
    FontSize:int=9
    __WorldContext:Object Reference
   ) -> (then, ContextOut:Paint Context Structure)
   > n12 = Create SliderLabel(
    OnPaintContext=@n11
    Slider=Get Slider_Inclination() -> (Slider_Inclination:Slider Object Reference)
    PositionalShift=(X=5,Y=8)
    Label=Adjust Sun Inclination
    Label Color=(R=1,G=1,B=1,A=1)
    FontSize=9
   )
   Draw Text!(
    execute=@n3.then_0
    Context:"Paint Context Structure (by ref)"=@n12
    Text:"Text (by ref)"=MakeLiteralText(Value="LOCTEXT(\"Directional Light Color\")")
    Position:Vector 2D Structure=vector2d + vector2d(A=@n10, B=(X=0,Y=-26))
    Font:Font Object Reference=&NotoSans
    FontSize:float="To Float (Integer)"(InInt=MakeLiteralInt!(Value:int=10) -> (ReturnValue:int=0))
    FontTypeFace:Name=Regular
    Tint:Linear Color Structure=(R=0,G=0.076,B=0,A=1)
   ) -> (then)
   > Draw Text(
    Context=@n12
    Text=MakeLiteralText(Value="LOCTEXT(\"Rayleigh Scatter Color\")")
    Position=vector2d + vector2d(A=@n10, B=(X=2,Y=-10))
    Font=&NotoSans
    FontSize="To Float (Integer)"(InInt=MakeLiteralInt(Value=10) -> (ReturnValue=0))
    FontTypeFace=Regular
    Tint=(R=0,G=0.0762,B=0,A=1)
   )
   n13 = vector2d + vector2d(A=vector2d + vector2d(A=@n7, B=(X=-10,Y=0)), B=@n4)
   DrawLine!(
    execute=@n3.then_1
    Context:"Paint Context Structure (by ref)"=@n12
    PositionA:Vector 2D Structure=@n9
    PositionB:Vector 2D Structure=@n13
    Tint:Linear Color Structure=(R=1,G=1,B=1,A=1)
    bAntiAlias:bool=true
    Thickness:float=2
   ) -> (then)
   > DrawLine(Context=@n12, PositionA=@n10, PositionB=@n9, Tint=(R=1,G=1,B=1,A=1), bAntiAlias=true, Thickness=1)
   n14 = MakeVector2D(X=2, Y=-16)
   > Draw Text(
    Context=@n12
    Text=MakeLiteralText(Value="LOCTEXT(\"Night\")")
    Position=vector2d + vector2d(A=@n9, B=@n14)
    Font=&NotoSans
    FontSize="To Float (Integer)"(InInt=MakeLiteralInt(Value=10) -> (ReturnValue=0))
    FontTypeFace=Regular
    Tint=(R=1,G=1,B=1,A=1)
   )
   n15 = vector2d - vector2d(A=@n13, B=@n4)
   > DrawLine(Context=@n12, PositionA=@n15, PositionB=@n13, Tint=(R=1,G=1,B=1,A=1), bAntiAlias=true, Thickness=1)
   n16 = Generate Line Along Bar Macro!(
    A:Vector 2D Structure=@n10
    B:Vector 2D Structure=@n15
    A2:Vector 2D Structure=@n9
    B2:Vector 2D Structure=@n13
    Alpha:double=float / float(A=3, B=24)
   ) -> (ReturnValue:Vector 2D Structure, ReturnValue2:Vector 2D Structure) {is=title K2Node_MacroInstance}
   > DrawLine(Context=@n12, PositionA=@n16.ReturnValue2, PositionB=@n16, Tint=(R=1,G=1,B=1,A=1), bAntiAlias=true, Thickness=1)
   > Draw Text(
    Context=@n12
    Text=MakeLiteralText(Value="LOCTEXT(\"Morning\")")
    Position=vector2d + vector2d(A=@n16.ReturnValue2, B=@n14)
    Font=&NotoSans
    FontSize="To Float (Integer)"(InInt=MakeLiteralInt(Value=10) -> (ReturnValue=0))
    FontTypeFace=Regular
    Tint=(R=1,G=1,B=1,A=1)
   )
   n17 = Generate Line Along Bar Macro(A=@n10, B=@n15, A2=@n9, B2=@n13, Alpha=float / float(A=5, B=24))
   > DrawLine(Context=@n12, PositionA=@n17.ReturnValue2, PositionB=@n17, Tint=(R=1,G=1,B=1,A=1), bAntiAlias=true, Thickness=1)
   > Draw Text(
    Context=@n12
    Text=MakeLiteralText(Value="LOCTEXT(\"Day\")")
    Position=vector2d + vector2d(A=@n17.ReturnValue2, B=@n14)
    Font=&NotoSans
    FontSize="To Float (Integer)"(InInt=MakeLiteralInt(Value=10) -> (ReturnValue=0))
    FontTypeFace=Regular
    Tint=(R=1,G=1,B=1,A=1)
   )
   n18 = Generate Line Along Bar Macro(A=@n10, B=@n15, A2=@n9, B2=@n13, Alpha=float / float(A=19, B=24))
   > DrawLine(Context=@n12, PositionA=@n18.ReturnValue2, PositionB=@n18, Tint=(R=1,G=1,B=1,A=1), bAntiAlias=true, Thickness=1)
   > Draw Text(
    Context=@n12
    Text=MakeLiteralText(Value="LOCTEXT(\"Evening\")")
    Position=vector2d + vector2d(A=@n18.ReturnValue2, B=@n14)
    Font=&NotoSans
    FontSize="To Float (Integer)"(InInt=MakeLiteralInt(Value=10) -> (ReturnValue=0))
    FontTypeFace=Regular
    Tint=(R=1,G=1,B=1,A=1)
   )
   n19 = Generate Line Along Bar Macro(A=@n10, B=@n15, A2=@n9, B2=@n13, Alpha=float / float(A=22, B=24))
   > DrawLine(Context=@n12, PositionA=@n19.ReturnValue2, PositionB=@n19, Tint=(R=1,G=1,B=1,A=1), bAntiAlias=true, Thickness=1)
   > Draw Text(
    Context=@n12
    Text=MakeLiteralText(Value="LOCTEXT(\"Night\")")
    Position=vector2d + vector2d(A=@n19.ReturnValue2, B=@n14)
    Font=&NotoSans
    FontSize="To Float (Integer)"(InInt=MakeLiteralInt(Value=10) -> (ReturnValue=0))
    FontTypeFace=Regular
    Tint=(R=1,G=1,B=1,A=1)
   )
   > Return Node!(execute)
   Reroute Node(InputPin:Vector 2D Structure=@n5) -> (OutputPin:Vector 2D Structure)
  ]}
 PrepForFalseSun {nodes [
   n1 = PrepForFalseSun!() -> (then, HasValidPieWorldRef:bool=~, WorldContextObject:Object Reference) {is=title K2Node_FunctionEntry}
   > n2 = Branch(Condition=@n1)
   > n3 = Sequence!(execute) -> (then_0, then_1, then_2:Exec, then_3:Exec, then_4:Exec)
   n4 = GetVectorParameterValue!(
    execute=@n3.then_1
    WorldContextObject
    Collection:Material Parameter Collection Object Reference=&FortniteMaterialParameters
    ParameterName:Name=SunlightYTransformVector
   ) -> (then, ReturnValue_R:float=~, ReturnValue_G:float=~, ReturnValue_B:float=~, ReturnValue_A:float=~) {is=title K2Node_CallMaterialParameterCollectionFunction}
   > Set SunY(SunY:Vector=MakeVector(X=@n4, Y=@n4.ReturnValue_G, Z=@n4.ReturnValue_B)) -> (Output_Get:Vector)
   n5 = GetVectorParameterValue(
    execute=@n3.then_2
    Collection=&FortniteMaterialParameters
    ParameterName=SunlightZTransformVector
   ) -> (ReturnValue_R=~, ReturnValue_G=~, ReturnValue_B=~, ReturnValue_A=~)
   > Set SunZ(SunZ:Vector=MakeVector(X=@n5, Y=@n5.ReturnValue_G, Z=@n5.ReturnValue_B)) -> (Output_Get:Vector)
   TODM_Off(
    execute=@n3.then_3
    self:BP Art Review Helper Chris Object Reference=[Get TODM BP Helper Dedicated Server Ref(), Get TODM BP Helper Client Ref()]
   )
   > PrintString(InString=TODM OFF)
   > Set Enabled And Tooltip Text(Widget=Get Button_ResetDirOverToSun(), Enabled=true, Tooltip Text="LOCTEXT(\"Click To Reset to Real Sun\")")
   Set UsingFalseSun?(execute=@n3.then_4, UsingFalseSun?=true)
   > Return Node()
   n6 = GetVectorParameterValue(
    execute=@n3.then_0
    WorldContextObject=@n1.WorldContextObject
    Collection=&FortniteMaterialParameters
    ParameterName=SunlightDirectionalVector
   ) -> (ReturnValue_R=~, ReturnValue_G=~, ReturnValue_B=~, ReturnValue_A=~)
   > n7 = Set SunDirection(SunDirection:Vector=MakeVector(X=@n6, Y=@n6.ReturnValue_G, Z=@n6.ReturnValue_B)) -> (Output_Get:Vector)
   > Set OrigionalSunDirection(OrigionalSunDirection:Vector=@n7) -> (Output_Get:Vector)
   PrintString(execute=@n2.else, InString="Error, No Valid Pie World For PrepForFalseSunFunction")
   Set Enabled And Tooltip Text(Widget=Get Slider_TimeOfDay(), Enabled=false, Tooltip Text="LOCTEXT(\"Reset To Real Sun To Change Time Of Day\")")
  ]}
 Update Bindings {nodes [
   n1 = Update Bindings!() -> (then, bIncludeDedicatedServer:bool=~) {is=title K2Node_FunctionEntry}
   > n2 = Sequence!(execute) -> (then_0, then_1, then_2)
   > PrintString(InString=Binding TODM, bPrintToScreen=false)
   > n3 = GetPIEWorlds!(execute, bIncludeDedicatedServer:bool=false) -> (then, ReturnValue:Array of World Object References)
   n4 = Is Array Non Empty!(InputArray:Array of World Object References=@n3) -> (Non-Empty:bool=~) {is=title K2Node_MacroInstance}
   > n5 = Branch(Condition=@n4)
   n6 = "Get (a copy)"(Array:Array of World Object References=@n3) -> (Output:World Object Reference)
   > n7 = GetActorOfClass!(
    execute
    WorldContextObject:Object Reference=@n6
    ActorClass:Actor Class Reference=/Game/ArtTools/Tech_Art/ArtReview_SunTool/SupportBPs/BR_Art_ReviewTODMSystem.BR_Art_ReviewTODMSystem_C
   ) -> (then, ReturnValue:BR Art Review TODMSystem Object Reference)
   > Set ToDManagerRef(ToDManagerRef:BR Art Review TODMSystem Object Reference=@n7) -> (Output_Get:BR Art Review TODMSystem Object Reference)
   n8 = GetActorOfClass(execute=@n5.else, ActorClass=&BR_Art_ReviewTODMSystem_C)
   > Set ToDManagerRef(ToDManagerRef=@n8)
   PrintString(execute=@n2.then_1, InString=Binding Helper, bPrintToScreen=false)
   > n9 = GetPIEWorlds(bIncludeDedicatedServer=@n1)
   > n10 = Branch(Condition=Is Array Non Empty(InputArray=@n9) -> (Non-Empty=~))
   > n11 = For Each Loop(Array:Array of World Object References=@n9) -> (Array Element:World Object Reference)
   > n12 = Branch(Condition=IsServer!(WorldContextObject:Object Reference=@n11) -> (ReturnValue:bool))
   > n13 = GetAllActorsOfClass!(
    execute
    WorldContextObject:Object Reference=@n11
    ActorClass:Actor Class Reference=/Game/ArtTools/Tech_Art/ArtReview_SunTool/SupportBPs/BP_ArtReviewHelper_Chris.BP_ArtReviewHelper_Chris_C
   ) -> (then, OutActors:Array of BP Art Review Helper Chris Object References)
   > n14 = Branch(Condition=Is Array Non Empty(InputArray:Array of BP Art Review Helper Chris Object References=@n13) -> (Non-Empty=~))
   > Set TODM BP Helper Dedicated Server Ref(TODM BP Helper Dedicated Server Ref:BP Art Review Helper Chris Object Reference="Get (a copy)"(Array:Array of BP Art Review Helper Chris Object References=@n13) -> (Output:BP Art Review Helper Chris Object Reference)) -> (Output_Get:BP Art Review Helper Chris Object Reference)
   PrintString(execute=@n14.else, InString="Error, SERVER WORLD: no BP HElPER FOUND!")
   n15 = GetAllActorsOfClass(
    execute=@n12.else
    WorldContextObject=@n11
    ActorClass=&BP_ArtReviewHelper_Chris_C
   )
   > Branch(Condition=Is Array Non Empty(InputArray=@n15) -> (Non-Empty=~))
   > Set TODM BP Helper Client Ref(TODM BP Helper Client Ref:BP Art Review Helper Chris Object Reference="Get (a copy)"(Array=@n15)) -> (Output_Get:BP Art Review Helper Chris Object Reference)
   n16 = GetAllActorsOfClass(execute=@n10.else, ActorClass=&BP_ArtReviewHelper_Chris_C)
   > n17 = Branch(Condition=Is Array Non Empty(InputArray=@n16) -> (Non-Empty=~))
   > n18 = Set TODM BP Helper Dedicated Server Ref(TODM BP Helper Dedicated Server Ref="Get (a copy)"(Array=@n16))
   > Set TODM BP Helper Client Ref(TODM BP Helper Client Ref=@n18)
   PrintString(execute=@n17.else, InString="Error, No Worlds, No Helper BP Found")
   Return Node!(execute=@n2.then_2, PieWorld:World Object Reference=@n6, PieWorldIsValid?:bool=@n4)
  ]}
 UpdateSliders {nodes [
   n1 = UpdateSliders!() -> (then, NewParam:ENUM_ArtReview_Sliders Enum=NewEnumerator0, World Object Reference:World Object Reference) {is=title K2Node_FunctionEntry}
   > n2 = Sequence!(execute) -> (then_0, then_1)
   > Branch(Condition=NOT Boolean(A="Equal (Enum)"(A:ENUM_ArtReview_Sliders Enum=@n1, B:ENUM_ArtReview_Sliders Enum=NewEnumerator0)))
   > SetValue(self=Get Slider_Rotation(), InValue=0)
   Branch(execute=@n2.then_1, Condition=NOT Boolean(A="Equal (Enum)"(A=@n1, B=NewEnumerator1)))
   > n3 = GetVectorParameterValue(
    WorldContextObject=@n1.World Object Reference
    Collection=&FortniteMaterialParameters
    ParameterName=SunlightDirectionalVector
   ) -> (ReturnValue_R=~, ReturnValue_G=~, ReturnValue_B=~, ReturnValue_A=~)
   n4 = MakeVector(X=@n3, Y=@n3.ReturnValue_G, Z=@n3.ReturnValue_B)
   > SetValue(self=Get Slider_Inclination(), InValue=float - float(A=90, B=GetAngleBetweenTwoVectors(VectorA=@n4, VectorB="0, 0,1")))
  ]}
 UpdateSun {nodes [
   n1 = UpdateSun!() -> (then, CallingSlider:ENUM_ArtReview_Sliders Enum=NewEnumerator0, SliderValue:double=~) {is=title K2Node_FunctionEntry}
   > n2 = Update Bindings!(execute, self, bIncludeDedicatedServer:bool=true) -> (then, PieWorld:World Object Reference, PieWorldIsValid?:bool) {is=title K2Node_CallFunction}
   > n3 = Switch on ENUM_ArtReview_Sliders!(execute, Selection:ENUM_ArtReview_Sliders Enum=@n1) -> (NewEnumerator0:Exec, NewEnumerator1:Exec, NewEnumerator2:Exec) {is=title K2Node_SwitchEnum}
   > n4 = Branch(Condition=Get UsingFalseSun?() -> (UsingFalseSun?:bool))
   n5 = Normalize(
    A=Rotate Vector Around Axis!(InVect:Vector=Get OrigionalSunDirection() -> (OrigionalSunDirection:Vector), AngleDeg:float=@n1.SliderValue, Axis:Vector="0, 0,1") -> (ReturnValue:Vector)
    Tolerance=0.0001
   )
   Set SunDirection(execute=[@n9.then, @n4.then], SunDirection=@n5)
   n6 = Normalize(A=Cross Product!(A:Vector="0, 0,1", B:Vector=@n5) -> (ReturnValue:Vector), Tolerance=0.0001)
   > Set SunZ(SunZ=@n6)
   n7 = Get SunDirection() -> (SunDirection:Vector)
   > Set SunY(SunY=Normalize(A=Cross Product(A=@n6, B=@n7), Tolerance=0.0001))
   n8 = Get ToDManagerRef()
   > UpdateSunOverrideValues!(
    execute
    self:BR Art Review TODMSystem Object Reference=@n8
    OverwritenSunDirectionalVector:Linear Color Structure="To LinearColor (Vector)"!(InVec:Vector=Get SunDirection()) -> (ReturnValue:Linear Color Structure)
    OverwrittenSunYDirection:Linear Color Structure="To LinearColor (Vector)"(InVec=Get SunY() -> (SunY:Vector))
    OverwrittenSunZDirection:Linear Color Structure="To LinearColor (Vector)"(InVec=Get SunZ() -> (SunZ:Vector))
   ) -> (then)
   > SetManualAnimeSunDirection(self=@n8, SwitchManualAnimeShadeSunDirection=true)
   > UpdateFalseRotation!(
    execute
    self:BP Art Review Helper Chris Object Reference=[Get TODM BP Helper Dedicated Server Ref(), Get TODM BP Helper Client Ref()]
    Forward Vector:Vector=Get SunDirection()
    SliderValue:double=@n1.SliderValue
   ) -> (then)
   n9 = PrepForFalseSun!(execute=@n4.else, self, HasValidPieWorldRef:bool=@n2.PieWorldIsValid?, WorldContextObject:Object Reference=@n2) -> (then) {is=title K2Node_CallFunction}
   n10 = Branch(execute=@n3.NewEnumerator1, Condition=Get UsingFalseSun?())
   n11 = BreakVector(InVec=Get SunDirection())
   n12 = MakeVector(X=@n11, Y=@n11.Y)
   n13 = Normalize(
    A="Lerp (Vector)"!(A:Vector=Normalize(A=@n12, Tolerance=0.0001), B:Vector="0, 0,1", Alpha:float=float / float(A=@n1.SliderValue, B=90)) -> (ReturnValue:Vector)
    Tolerance=0.0001
   )
   Set OrigionalSunDirection(execute=[@n17.then, @n10.then], OrigionalSunDirection=@n13)
   > n14 = Set SunDirection(SunDirection=@n13)
   n15 = Get SunZ()
   > Set SunY(SunY=Normalize(A=Cross Product(A=@n15, B=@n14), Tolerance=0.0001))
   n16 = Get ToDManagerRef()
   > UpdateSunOverrideValues(
    self=@n16
    OverwritenSunDirectionalVector="To LinearColor (Vector)"(InVec=Get SunDirection())
    OverwrittenSunYDirection="To LinearColor (Vector)"(InVec=Get SunY())
    OverwrittenSunZDirection="To LinearColor (Vector)"(InVec=Get SunZ())
   )
   > SetManualAnimeSunDirection(self=@n16, SwitchManualAnimeShadeSunDirection=true)
   > SetValue(self=Get Slider_Rotation(), InValue=0)
   > UpdateAltitudeRotation!(
    execute
    self:BP Art Review Helper Chris Object Reference=[Get TODM BP Helper Dedicated Server Ref(), Get TODM BP Helper Client Ref()]
    Rotation:double=@n1.SliderValue
   ) -> (then)
   n17 = PrepForFalseSun(execute=@n10.else, HasValidPieWorldRef=@n2.PieWorldIsValid?, WorldContextObject=@n2)
   n18 = Branch(execute=@n3.NewEnumerator2, Condition=Get UsingFalseSun?())
   > n19 = ResetToRealSun() {is=title K2Node_CallFunction}
   n20 = Branch(execute=[@n19.then, @n18.else])
   UpdateHour!(
    execute=[@n22.then, @n20.then]
    self:BP Art Review Helper Chris Object Reference=Get TODM BP Helper Dedicated Server Ref()
    hour:double=@n1.SliderValue
   ) -> (then)
   > UpdateSliders!(execute, self, NewParam:ENUM_ArtReview_Sliders Enum=NewEnumerator0, World Object Reference:World Object Reference=@n2) -> (then) {is=title K2Node_CallFunction}
   > EditorInvalidateViewports(self:Level Editor Subsystem Object Reference=Get LevelEditorSubsystem())
   > SetFocusToGameViewport!(execute) -> (then)
   n21 = GetCurrentDayPhase!(WorldContextObject:Object Reference=Get ToDManagerRef()) -> (ReturnValue:EFortDayPhase Enum=Morning)
   n22 = PrintString(
    execute=@n20.else
    InString=Append!(
     A="Sun Slider Debig:  Hour - "
     B="To String (Float)"(InDouble=@n1.SliderValue)
     C:String=" : "
     D:String=Enum to String!(Enumerator:EFortDayPhase Enum=@n21) -> (ReturnValue:String)
    ) -> (ReturnValue)
   )
  ]}
]
variables [
 HideLabels {type=boolean}
 OrigionalSunDirection {type=Vector}
 SunColors {
  type="array<LinearColor>"
  default [
   <rgba 0 0 0 0>
   <rgba 0 0 0 0>
   <rgba 0 0 0 0>
   <rgba 0 0 0 0>
   <rgba 0 0 0 0>
   <rgba 0 0 0 0>
   <rgba 0 0 0 0>
   <rgba 0 0 0 0>
   <rgba 0 0 0 0>
   <rgba 0 0 0 0>
   <rgba 0 0 0 0>
  ]
 }
 SunDirection {type=Vector}
 SunY {type=Vector}
 SunZ {type=Vector}
 TODM BP Helper Client Ref {type=BP_ArtReviewHelper_Chris_C, default=null}
 TODM BP Helper Dedicated Server Ref {type=BP_ArtReviewHelper_Chris_C, default=null}
 ToDManagerRef {type=BR_Art_ReviewTODMSystem_C, default=null}
 UsingFalseSun? {type=boolean}
]
widgets {
 info {rootWidgetClass=&Overlay, widgetCount=7}
 widgets [
  Overlay_26 {widgetClassPath=&Overlay, widget=Overlay_26}
  VerticalBox_0 {
   widgetClassPath=&VerticalBox
   parent=Overlay_26
   widget=VerticalBox_0
   slot {class=OverlaySlot, properties {horizontalAlignment=HAlign_Fill}}
  }
  HorizontalBox_SunHourControls {
   widgetClassPath=&HorizontalBox
   bIsVariable=true
   parent=VerticalBox_0
   widget=HorizontalBox_SunHourControls
   slot {class=VerticalBoxSlot, properties {padding=<ltrb 0 0 0 32>, size {sizeRule=Fill, value=1}}}
  }
  Button_ResetDirOverToSun {
   widgetClassPath=&TA_Generic01_Button
   parent=HorizontalBox_SunHourControls
   widget=Button_ResetDirOverToSun
   properties {isEnabled=false, label=Reset Sun}
   slot {class=HorizontalBoxSlot}
  }
  Slider_TimeOfDay {
   widgetClassPath=/Script/UMG.Slider
   bIsVariable=true
   parent=HorizontalBox_SunHourControls
   widget=Slider_TimeOfDay
   properties {
    maxValue=24
    toolTipText=Real Sun Time of Day
    value=12
    widgetStyle {
     barThickness=4
     disabledBarImage {
      bIsDynamicallyLoaded=false
      drawAs=RoundedBox
      imageType=NoImage
      mirroring=NoMirror
      resourceName=None
      resourceObject=null
      tiling=NoTile
      imageSize=<32 32>
      margin=<ltrb 0 0 0 0>
      outlineSettings {
       bUseBrushTransparency=false
       roundingType=FixedRadius
       width=1
       color {colorUseRule=UseColor_ColorTable, specifiedColor=<rgba 0 0 0 0>}
       cornerRadii=<2 2 2 2>
      }
      tintColor {colorUseRule=UseColor_Specified, specifiedColor=<rgba 1 1 1 1>}
      uVRegion {bIsValid=false, max=<0 0>, min=<0 0>}
     }
     disabledThumbImage {
      bIsDynamicallyLoaded=false
      drawAs=RoundedBox
      imageType=NoImage
      mirroring=NoMirror
      resourceName=None
      resourceObject=null
      tiling=NoTile
      imageSize=<32 32>
      margin=<ltrb 0 0 0 0>
      outlineSettings {
       bUseBrushTransparency=false
       roundingType=HalfHeightRadius
       width=0
       color {colorUseRule=UseColor_Specified, specifiedColor=<rgba 0 0 0 0>}
       cornerRadii=<0 0 0 0>
      }
      tintColor {colorUseRule=UseColor_Specified, specifiedColor=<rgba 0.5271 0.5271 0.5271 1>}
      uVRegion {bIsValid=false, max=<0 0>, min=<0 0>}
     }
     hoveredBarImage {
      bIsDynamicallyLoaded=false
      drawAs=RoundedBox
      imageType=NoImage
      mirroring=NoMirror
      resourceName=None
      resourceObject=null
      tiling=NoTile
      imageSize=<0 0>
      margin=<ltrb 0 0 0 0>
      outlineSettings {
       bUseBrushTransparency=false
       roundingType=FixedRadius
       width=1
       color {colorUseRule=UseColor_ColorTable, specifiedColor=<rgba 0 0 0 0>}
       cornerRadii=<2 2 2 2>
      }
      tintColor {colorUseRule=UseColor_Specified, specifiedColor=<rgba 0.0048 0.0048 0.0048 1>}
      uVRegion {bIsValid=false, max=<0 0>, min=<0 0>}
     }
     hoveredThumbImage {
      bIsDynamicallyLoaded=false
      drawAs=RoundedBox
      imageType=NoImage
      mirroring=NoMirror
      resourceName=None
      resourceObject=null
      tiling=NoTile
      imageSize=<8 8>
      margin=<ltrb 0 0 0 0>
      outlineSettings {
       bUseBrushTransparency=false
       roundingType=HalfHeightRadius
       width=0
       color {colorUseRule=UseColor_Specified, specifiedColor=<rgba 0 0 0 0>}
       cornerRadii=<0 0 0 0>
      }
      tintColor {colorUseRule=UseColor_Specified, specifiedColor=<rgba 1 1 1 1>}
      uVRegion {bIsValid=false, max=<0 0>, min=<0 0>}
     }
     normalBarImage {
      bIsDynamicallyLoaded=false
      drawAs=RoundedBox
      imageType=NoImage
      mirroring=NoMirror
      resourceName=None
      resourceObject=null
      tiling=NoTile
      imageSize=<0 0>
      margin=<ltrb 0 0 0 0>
      outlineSettings {
       bUseBrushTransparency=false
       roundingType=FixedRadius
       width=1
       color {colorUseRule=UseColor_ColorTable, specifiedColor=<rgba 0 0 0 0>}
       cornerRadii=<2 2 2 2>
      }
      tintColor {colorUseRule=UseColor_Specified, specifiedColor=<rgba 0.0048 0.0048 0.0048 1>}
      uVRegion {bIsValid=false, max=<0 0>, min=<0 0>}
     }
     normalThumbImage {
      bIsDynamicallyLoaded=false
      drawAs=RoundedBox
      imageType=NoImage
      mirroring=NoMirror
      resourceName=None
      resourceObject=null
      tiling=NoTile
      imageSize=<8 8>
      margin=<ltrb 0 0 0 0>
      outlineSettings {
       bUseBrushTransparency=false
       roundingType=HalfHeightRadius
       width=0
       color {colorUseRule=UseColor_Specified, specifiedColor=<rgba 0 0 0 0>}
       cornerRadii=<0 0 0 0>
      }
      tintColor {colorUseRule=UseColor_Specified, specifiedColor=<rgba 0.2159 0.2159 0.2159 1>}
      uVRegion {bIsValid=false, max=<0 0>, min=<0 0>}
     }
    }
   }
   slot {class=HorizontalBoxSlot, properties {size {sizeRule=Fill, value=1}}}
  }
  Slider_Rotation {
   widgetClassPath=&Slider
   bIsVariable=true
   parent=VerticalBox_0
   widget=Slider_Rotation
   properties {
    maxValue=180
    minValue=-180
    toolTipText=False Sun Rotation
    widgetStyle {
     barThickness=4
     disabledBarImage {
      bIsDynamicallyLoaded=false
      drawAs=RoundedBox
      imageType=NoImage
      mirroring=NoMirror
      resourceName=None
      resourceObject=null
      tiling=NoTile
      imageSize=<32 32>
      margin=<ltrb 0 0 0 0>
      outlineSettings {
       bUseBrushTransparency=false
       roundingType=FixedRadius
       width=1
       color {colorUseRule=UseColor_ColorTable, specifiedColor=<rgba 0 0 0 0>}
       cornerRadii=<2 2 2 2>
      }
      tintColor {colorUseRule=UseColor_Specified, specifiedColor=<rgba 1 1 1 1>}
      uVRegion {bIsValid=false, max=<0 0>, min=<0 0>}
     }
     disabledThumbImage {
      bIsDynamicallyLoaded=false
      drawAs=RoundedBox
      imageType=NoImage
      mirroring=NoMirror
      resourceName=None
      resourceObject=null
      tiling=NoTile
      imageSize=<32 32>
      margin=<ltrb 0 0 0 0>
      outlineSettings {
       bUseBrushTransparency=false
       roundingType=HalfHeightRadius
       width=0
       color {colorUseRule=UseColor_Specified, specifiedColor=<rgba 0 0 0 0>}
       cornerRadii=<0 0 0 0>
      }
      tintColor {colorUseRule=UseColor_Specified, specifiedColor=<rgba 0.5271 0.5271 0.5271 1>}
      uVRegion {bIsValid=false, max=<0 0>, min=<0 0>}
     }
     hoveredBarImage {
      bIsDynamicallyLoaded=false
      drawAs=RoundedBox
      imageType=NoImage
      mirroring=NoMirror
      resourceName=None
      resourceObject=null
      tiling=NoTile
      imageSize=<0 0>
      margin=<ltrb 0 0 0 0>
      outlineSettings {
       bUseBrushTransparency=false
       roundingType=FixedRadius
       width=1
       color {colorUseRule=UseColor_ColorTable, specifiedColor=<rgba 0 0 0 0>}
       cornerRadii=<2 2 2 2>
      }
      tintColor {colorUseRule=UseColor_Specified, specifiedColor=<rgba 0.0048 0.0048 0.0048 1>}
      uVRegion {bIsValid=false, max=<0 0>, min=<0 0>}
     }
     hoveredThumbImage {
      bIsDynamicallyLoaded=false
      drawAs=RoundedBox
      imageType=NoImage
      mirroring=NoMirror
      resourceName=None
      resourceObject=null
      tiling=NoTile
      imageSize=<8 8>
      margin=<ltrb 0 0 0 0>
      outlineSettings {
       bUseBrushTransparency=false
       roundingType=HalfHeightRadius
       width=0
       color {colorUseRule=UseColor_Specified, specifiedColor=<rgba 0 0 0 0>}
       cornerRadii=<0 0 0 0>
      }
      tintColor {colorUseRule=UseColor_Specified, specifiedColor=<rgba 1 1 1 1>}
      uVRegion {bIsValid=false, max=<0 0>, min=<0 0>}
     }
     normalBarImage {
      bIsDynamicallyLoaded=false
      drawAs=RoundedBox
      imageType=NoImage
      mirroring=NoMirror
      resourceName=None
      resourceObject=null
      tiling=NoTile
      imageSize=<0 0>
      margin=<ltrb 0 0 0 0>
      outlineSettings {
       bUseBrushTransparency=false
       roundingType=FixedRadius
       width=1
       color {colorUseRule=UseColor_ColorTable, specifiedColor=<rgba 0 0 0 0>}
       cornerRadii=<2 2 2 2>
      }
      tintColor {colorUseRule=UseColor_Specified, specifiedColor=<rgba 0.0048 0.0048 0.0048 1>}
      uVRegion {bIsValid=false, max=<0 0>, min=<0 0>}
     }
     normalThumbImage {
      bIsDynamicallyLoaded=false
      drawAs=RoundedBox
      imageType=NoImage
      mirroring=NoMirror
      resourceName=None
      resourceObject=null
      tiling=NoTile
      imageSize=<8 8>
      margin=<ltrb 0 0 0 0>
      outlineSettings {
       bUseBrushTransparency=false
       roundingType=HalfHeightRadius
       width=0
       color {colorUseRule=UseColor_Specified, specifiedColor=<rgba 0 0 0 0>}
       cornerRadii=<0 0 0 0>
      }
      tintColor {colorUseRule=UseColor_Specified, specifiedColor=<rgba 0.2159 0.2159 0.2159 1>}
      uVRegion {bIsValid=false, max=<0 0>, min=<0 0>}
     }
    }
   }
   slot {class=VerticalBoxSlot, properties {verticalAlignment=VAlign_Bottom, size {sizeRule=Fill, value=1}}}
  }
  Slider_Inclination {
   widgetClassPath=&Slider
   bIsVariable=true
   parent=VerticalBox_0
   widget=Slider_Inclination
   properties {
    maxValue=89.9
    toolTipText=False Sun Inclination
    widgetStyle {
     barThickness=4
     disabledBarImage {
      bIsDynamicallyLoaded=false
      drawAs=RoundedBox
      imageType=NoImage
      mirroring=NoMirror
      resourceName=None
      resourceObject=null
      tiling=NoTile
      imageSize=<32 32>
      margin=<ltrb 0 0 0 0>
      outlineSettings {
       bUseBrushTransparency=false
       roundingType=FixedRadius
       width=1
       color {colorUseRule=UseColor_ColorTable, specifiedColor=<rgba 0 0 0 0>}
       cornerRadii=<2 2 2 2>
      }
      tintColor {colorUseRule=UseColor_Specified, specifiedColor=<rgba 1 1 1 1>}
      uVRegion {bIsValid=false, max=<0 0>, min=<0 0>}
     }
     disabledThumbImage {
      bIsDynamicallyLoaded=false
      drawAs=RoundedBox
      imageType=NoImage
      mirroring=NoMirror
      resourceName=None
      resourceObject=null
      tiling=NoTile
      imageSize=<32 32>
      margin=<ltrb 0 0 0 0>
      outlineSettings {
       bUseBrushTransparency=false
       roundingType=HalfHeightRadius
       width=0
       color {colorUseRule=UseColor_Specified, specifiedColor=<rgba 0 0 0 0>}
       cornerRadii=<0 0 0 0>
      }
      tintColor {colorUseRule=UseColor_Specified, specifiedColor=<rgba 0.5271 0.5271 0.5271 1>}
      uVRegion {bIsValid=false, max=<0 0>, min=<0 0>}
     }
     hoveredBarImage {
      bIsDynamicallyLoaded=false
      drawAs=RoundedBox
      imageType=NoImage
      mirroring=NoMirror
      resourceName=None
      resourceObject=null
      tiling=NoTile
      imageSize=<0 0>
      margin=<ltrb 0 0 0 0>
      outlineSettings {
       bUseBrushTransparency=false
       roundingType=FixedRadius
       width=1
       color {colorUseRule=UseColor_ColorTable, specifiedColor=<rgba 0 0 0 0>}
       cornerRadii=<2 2 2 2>
      }
      tintColor {colorUseRule=UseColor_Specified, specifiedColor=<rgba 0.0048 0.0048 0.0048 1>}
      uVRegion {bIsValid=false, max=<0 0>, min=<0 0>}
     }
     hoveredThumbImage {
      bIsDynamicallyLoaded=false
      drawAs=RoundedBox
      imageType=NoImage
      mirroring=NoMirror
      resourceName=None
      resourceObject=null
      tiling=NoTile
      imageSize=<8 8>
      margin=<ltrb 0 0 0 0>
      outlineSettings {
       bUseBrushTransparency=false
       roundingType=HalfHeightRadius
       width=0
       color {colorUseRule=UseColor_Specified, specifiedColor=<rgba 0 0 0 0>}
       cornerRadii=<0 0 0 0>
      }
      tintColor {colorUseRule=UseColor_Specified, specifiedColor=<rgba 1 1 1 1>}
      uVRegion {bIsValid=false, max=<0 0>, min=<0 0>}
     }
     normalBarImage {
      bIsDynamicallyLoaded=false
      drawAs=RoundedBox
      imageType=NoImage
      mirroring=NoMirror
      resourceName=None
      resourceObject=null
      tiling=NoTile
      imageSize=<0 0>
      margin=<ltrb 0 0 0 0>
      outlineSettings {
       bUseBrushTransparency=false
       roundingType=FixedRadius
       width=1
       color {colorUseRule=UseColor_ColorTable, specifiedColor=<rgba 0 0 0 0>}
       cornerRadii=<2 2 2 2>
      }
      tintColor {colorUseRule=UseColor_Specified, specifiedColor=<rgba 0.0048 0.0048 0.0048 1>}
      uVRegion {bIsValid=false, max=<0 0>, min=<0 0>}
     }
     normalThumbImage {
      bIsDynamicallyLoaded=false
      drawAs=RoundedBox
      imageType=NoImage
      mirroring=NoMirror
      resourceName=None
      resourceObject=null
      tiling=NoTile
      imageSize=<8 8>
      margin=<ltrb 0 0 0 0>
      outlineSettings {
       bUseBrushTransparency=false
       roundingType=HalfHeightRadius
       width=0
       color {colorUseRule=UseColor_Specified, specifiedColor=<rgba 0 0 0 0>}
       cornerRadii=<0 0 0 0>
      }
      tintColor {colorUseRule=UseColor_Specified, specifiedColor=<rgba 0.2159 0.2159 0.2159 1>}
      uVRegion {bIsValid=false, max=<0 0>, min=<0 0>}
     }
    }
   }
   slot {class=VerticalBoxSlot, properties {size {sizeRule=Fill, value=1}}}
  }
 ]
}
